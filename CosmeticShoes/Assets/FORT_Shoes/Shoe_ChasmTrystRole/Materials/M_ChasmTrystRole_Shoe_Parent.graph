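# Parent material for the ChasmTrystRole cosmetic shoe, adding a panning heart pattern to the shoe surface.
# It builds on the 2023 character-parent PreFX and PostFX functions, with MF_Sock in the chain.
# A combined mask texture on UV1 supplies a static shoe mask and tiled, scrolling hearts that modulate base color, roughness and specular.
# Heart normals derive Z from packed XY, are reoriented via DeriveTangentBasis and blend angle-corrected into the base normal.
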Material /CosmeticShoes/Assets/FORT_Shoes/Shoe_ChasmTrystRole/Materials/M_ChasmTrystRole_Shoe_Parent
graph {
 properties {
  bUseMaterialAttributes=true
  bUsedWithClothing=true
  bUsedWithMeshDeformer=true
  bUsedWithMorphTargets=true
  bUsedWithNanite=true
  bUsedWithNiagaraMeshParticles=true
  bUsedWithSkeletalMesh=true
  pixelDepthOffsetMode=PDOM_Legacy
  shadingModel=MSM_FromMaterialExpression
  usedShadingModels=Default Lit | Subsurface | Clear Coat | Subsurface Profile | Cloth
 }
 nodes [
  n1 = MaterialFunctionCall() {materialFunction=&MF_CharacterParent_2023_PreFX}
  MaterialFunctionCall(
   Material=MaterialFunctionCall!(InOpacity, Material=NamedRerouteUsage(), Mask=@n1.Tech Art Mask) {materialFunction=/Game/Global/Materials/Cosmetics/Kicks/Socks/Materials/MF_Sock}.Result
   M=@n1.M
   Pre Skinned Local Position=@n1.M
   Distance Blend=@n1.M
   Fresnel=@n1.M
  ) {materialFunction=&MF_CharacterParent_2023_PostFX}
  n2 = NamedRerouteUsage()
  MaterialFunctionCall(Normal=@n2, Flatness=Constant() {r=0.35}) {materialFunction=&FlattenNormal}
  n3 = TextureCoordinate() {coordinateIndex=1}
  n4 = ScalarParameter() {defaultValue=25, group=HeartPattern, parameterName=HeartPattern_Tiling}
  n5 = Constant()
  n6 = ScalarParameter() {defaultValue=-0.6, group=HeartPattern, parameterName=HeartPattern_PanningSpeed}
  n7 = TextureSample(UVs=Panner(Coordinate=Multiply(A=@n3, B=@n4) {constB=5}, Speed=AppendVector(A=@n5, B=@n6)) {speedY=0.1}) {
   samplerSource=SSM_Wrap_WorldGroupSettings
   samplerType=SAMPLERTYPE_LinearColor
   texture=/CosmeticShoes/Assets/FORT_Shoes/Shoe_ChasmTrystRole/Textures/T_ChasmTrystRole_Shoe_masks_combined
  }
  n8 = TextureSample() {
   constCoordinate=1
   samplerSource=SSM_Wrap_WorldGroupSettings
   samplerType=SAMPLERTYPE_LinearColor
   texture=&T_ChasmTrystRole_Shoe_masks_combined
  }
  NamedRerouteDeclaration(None=Multiply(A=SmoothStep(Value=@n7.G) {constMax=0.8, constMin=0.2}, B=NamedRerouteDeclaration(None=@n8.R) {name=ShoeMask, nodeColor=<rgba 0 0.7891 1 1>})) {name=HeartsPattern_Mask, nodeColor=<rgba 0.8281 0 1 1>}
  NamedRerouteDeclaration(None=@n1.Result) {name=InMA, nodeColor=<rgba 0 1 0.5078 1>}
  n9 = GetMaterialAttributes(None=NamedRerouteUsage()) {attributeGetTypes [
    #1
    #2
    #3
    #4
   ]}
  n10 = ScalarParameter() {defaultValue=0.85, group=HeartPattern, parameterName=HeartPattern_BaseColorMultiplier, sortPriority=50}
  n11 = NamedRerouteUsage()
  n12 = Multiply(A=ScalarParameter() {defaultValue=-0.2, group=HeartPattern, parameterName=HeartPattern_RoughnessOffset, sortPriority=50}, B=@n11)
  n13 = Multiply(A=@n11, B=ScalarParameter() {defaultValue=0.3, group=HeartPattern, parameterName=HeartPattern_SpecularOffset, sortPriority=50})
  n14 = NamedRerouteUsage()
  n15 = LinearInterpolate(A=TextureCoordinate(), B=TextureCoordinate() {coordinateIndex=1}, Alpha=@n14)
  n16 = Transform(None=Add(
    A=Transform(None=@n9.Normal)
    B=MaterialFunctionCall!(UVs=@n15, Normal=@n2) {materialFunction=/Engine/Functions/Engine_MaterialFunctions01/Texturing/DeriveTangentBasis}.World Space Normal
   )) {transformSourceType=TRANSFORMSOURCE_World, transformType=TRANSFORM_Tangent}
  n17 = SetMaterialAttributes!(
   MaterialAttributes=@n9.MaterialAttributes
   Base Color=LinearInterpolate(A=@n9.Base Color, B=Multiply(A=@n9.Base Color, B=@n10) {constB=0.5}, Alpha=@n11)
   Roughness=Saturate(None=Add(A=@n9.Roughness, B=@n12))
   Specular=Saturate(None=Add(A=@n9.Specular, B=@n13))
   Normal=LinearInterpolate(
    A=@n9.Normal
    B=MaterialFunctionCall(BaseNormal=@n9.Normal, AdditionalNormal=@n16) {materialFunction=&BlendAngleCorrectedNormals}.Result
    Alpha=@n14
   )
  ) {attributeSetTypes [
    #1
    #2
    #3
    #4
   ]}
  NamedRerouteDeclaration(None=@n17) {name=OutMA, nodeColor=<rgba 1 0.1875 0 1>}
  NamedRerouteDeclaration(None=AppendVector(A=@n7.B, B=@n7.B)) {name=HeartNormal_XY, nodeColor=<rgba 0 0.3871 1 1>}
  n18 = ConstantBiasScale(None=NamedRerouteUsage()) {bias=-0.5, scale=2}
  n19 = Multiply(A=@n18, B=@n18)
  n20 = Subtract(A=Subtract(B=ComponentMask(None=@n19) {r=true}), B=ComponentMask(None=@n19) {g=true})
  NamedRerouteDeclaration(None=AppendVector(A=@n18, B=SquareRoot(None=@n20))) {name=HeartNormal, nodeColor=<rgba 0 1 0.7422 1>}
 ]
}
